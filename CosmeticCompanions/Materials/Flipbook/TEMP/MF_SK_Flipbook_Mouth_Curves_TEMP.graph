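# Temporary material function that picks a mouth flipbook frame index for Cosmetic Companions from animation-curve-driven scalar parameters.
# Smile, frown, jaw-open and mouth emote curves each feed an If node with a threshold parameter.
# Constant indices cover neutral plus those four poses.
# The If results are passed through named reroutes, and the function returns a single Index output.
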MaterialFunction /CosmeticCompanions/Materials/Flipbook/TEMP/MF_SK_Flipbook_Mouth_Curves_TEMP
graph {nodes [
  FunctionOutput(None=NamedRerouteUsage()) {outputName=Index}
  NamedRerouteDeclaration(None=ScalarParameter() {group="Flipbook 04.01 Mouth | Curves (Driven By Animation)", parameterName=R_frown_pose, sortPriority=1}) {name=Frown_Pose, nodeColor=<rgba 1 0.2105 0 1>}
  NamedRerouteDeclaration(None=ScalarParameter() {defaultValue=0.1, group=Flipbook 04.02 Mouth | Thresholds, parameterName=Threshold_Frown, sortPriority=1}) {name=Threshold_Frown, nodeColor=<rgba 1 0.0834 0.4399 1>}
  NamedRerouteDeclaration(None=ScalarParameter() {group="Flipbook 04.01 Mouth | Curves (Driven By Animation)", parameterName=R_smile_pose, sortPriority=0}) {name=Smile_Pose, nodeColor=<rgba 1 0.2105 0 1>}
  NamedRerouteDeclaration(None=ScalarParameter() {group="Flipbook 04.01 Mouth | Curves (Driven By Animation)", parameterName=jaw_open_pose, sortPriority=2}) {name=Jaw_Open_Pose, nodeColor=<rgba 1 0.2105 0 1>}
  NamedRerouteDeclaration(None=Constant()) {name=Index_Neutral, nodeColor=<rgba 0.2818 0.3488 1 1>}
  NamedRerouteDeclaration(None=Constant() {r=3}) {name=Index_Jaw_Open, nodeColor=<rgba 0.2818 0.3488 1 1>}
  NamedRerouteDeclaration(None=Constant() {r=1}) {name=Index_Smile, nodeColor=<rgba 0.2818 0.3488 1 1>}
  NamedRerouteDeclaration(None=Constant() {r=2}) {name=Index_Frown, nodeColor=<rgba 0.2818 0.3488 1 1>}
  NamedRerouteDeclaration(None=If(A=NamedRerouteUsage(), B=NamedRerouteUsage(), "A > B"=NamedRerouteUsage(), "A < B"=NamedRerouteUsage())) {name=Smile_Output, nodeColor=<rgba 0 0.1532 1 1>}
  NamedRerouteDeclaration(None=If(A=NamedRerouteUsage(), B=NamedRerouteUsage(), "A > B"=NamedRerouteUsage(), "A < B"=NamedRerouteUsage())) {name=Frown_Output, nodeColor=<rgba 0 0.1532 1 1>}
  NamedRerouteDeclaration(None=If(A=NamedRerouteUsage(), B=NamedRerouteUsage(), "A > B"=NamedRerouteUsage(), "A < B"=NamedRerouteUsage())) {name=Jaw_Open_Output, nodeColor=<rgba 0 0.1532 1 1>}
  NamedRerouteDeclaration(None=ScalarParameter() {group="Flipbook 04.01 Mouth | Curves (Driven By Animation)", parameterName=mouth_emote_01, sortPriority=2}) {name=Mouth_Emote_01, nodeColor=<rgba 1 0.2105 0 1>}
  NamedRerouteDeclaration(None=ScalarParameter() {defaultValue=0.1, group=Flipbook 04.02 Mouth | Thresholds, parameterName=Threshold_Mouth_Emote_01, sortPriority=2}) {name=Threshold_Mouth_Emote_01, nodeColor=<rgba 1 0.0834 0.4399 1>}
  NamedRerouteDeclaration(None=Constant() {r=4}) {name=Index_Mouth_Emote_01, nodeColor=<rgba 0.2818 0.3488 1 1>}
  NamedRerouteDeclaration(None=If(A=NamedRerouteUsage(), B=NamedRerouteUsage(), "A > B"=NamedRerouteUsage(), "A < B"=NamedRerouteUsage())) {name=Mouth_Emote_Output, nodeColor=<rgba 0 0.1532 1 1>}
  NamedRerouteDeclaration(None=ScalarParameter() {defaultValue=0.1, group=Flipbook 04.02 Mouth | Thresholds, parameterName=Threshold_Smile, sortPriority=0}) {name=Threshold_Smile, nodeColor=<rgba 1 0.0834 0.4399 1>}
  NamedRerouteDeclaration(None=ScalarParameter() {defaultValue=0.1, group=Flipbook 04.02 Mouth | Thresholds, parameterName=Threshold_Jaw_Open, sortPriority=2}) {name=Threshold_Jaw_Open, nodeColor=<rgba 1 0.0834 0.4399 1>}
 ]}
