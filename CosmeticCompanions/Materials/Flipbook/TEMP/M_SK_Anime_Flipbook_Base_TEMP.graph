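# Temporary base material for skeletal-mesh cosmetic companions, combining anime cel shading with flipbook faces.
# Curve functions give frame indices for brows, eyes and mouth, which are remapped to sample flipbook textures blended into lit/shaded color, emissive and opacity.
# Vertex-color masks drive world position offset to hide face geometry.
# Static switches select features, including eye animation that overrides flipbook eyes, brows-as-eyes and debug views; shading normals use a sphere-normal function.
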Material /CosmeticCompanions/Materials/Flipbook/TEMP/M_SK_Anime_Flipbook_Base_TEMP
graph {
 properties {
  bAutomaticallySetUsageInEditor=false
  bUseMaterialAttributes=true
  bUsedWithClothing=true
  bUsedWithMeshDeformer=true
  bUsedWithMorphTargets=true
  bUsedWithNiagaraMeshParticles=true
  bUsedWithSkeletalMesh=true
  bUsedWithStaticLighting=true
  pixelDepthOffsetMode=PDOM_Legacy
  shadingModel=MSM_Cloth
  usedShadingModels=Cloth
 }
 nodes [
  n1 = VectorParameter() {
   group=Flipbook 04 Mouth
   parameterName=Flipbook.MouthRemap_4-7
   sortPriority=201
   channelNames {a=7 - Unused, b=6 - Unused, g=5 - Unused, r=4 - Emote}
   defaultValue=<rgba 0 1 2 3>
  }
  NamedRerouteDeclaration(None=@n1.RGBA) {name=Mouth_Remap_4-7, nodeColor=<rgba 1 0.6652 0.0014 1>}
  NamedRerouteDeclaration(None=ChannelMaskParameter(None=VertexColor()) {group=Flipbook 02 Brows, parameterName=Flipbook.BrowsVertexColor}.R) {name=Brows_VertexColor, nodeColor=<rgba 1 0.9844 0 1>}
  NamedRerouteDeclaration(None=StaticBoolParameter() {defaultValue=true, group=Flipbook 01 Global, parameterName=Flipbook.UseMouth, sortPriority=-8}) {name=UseMouth, nodeColor=<rgba 0.1953 1 0 1>}
  n2 = MaterialFunctionCall!(
   Curve Index=NamedRerouteUsage()
   Use Texture=NamedRerouteUsage()
   ----------------------
   Vertex Color=NamedRerouteUsage()
   ---------------------
   Flipbook Texture=NamedRerouteUsage()
   Column Count=NamedRerouteUsage()
   Row Count=NamedRerouteUsage()
   UVs=NamedRerouteUsage()
   UV Offset=NamedRerouteUsage()
   UV Scale=NamedRerouteUsage()
  ) {materialFunction=/Game/Global/Materials/Character_Parent_2023/FlipbookFace_Parent_2023/MaterialFunctions/MF_Sample_Flipbook}
  NamedRerouteDeclaration(None=@n2.Flipbook Vertex Mask) {name=MouthVertexMask, nodeColor=<rgba 0.2092 0.8266 1 1>}
  NamedRerouteDeclaration(None=@n2.Flipbook Color) {name=MouthColor, nodeColor=<rgba 0.2092 0.8266 1 1>}
  NamedRerouteDeclaration(None=@n2.Flipbook Alpha) {name=MouthAlpha, nodeColor=<rgba 0.2092 0.8266 1 1>}
  n3 = NamedRerouteUsage()
  n4 = Constant()
  n5 = NamedRerouteUsage()
  n6 = Add(
   A=StaticSwitch(True=NamedRerouteUsage(), False=Constant(), Value=NamedRerouteUsage())
   B=StaticSwitch(True=Constant(), False=StaticSwitch(True=@n3, False=@n4, Value=@n5), Value=NamedRerouteUsage())
  )
  n7 = StaticSwitch(True=NamedRerouteUsage(), False=Constant(), Value=NamedRerouteUsage())
  n8 = NamedRerouteUsage()
  n9 = Constant()
  n10 = NamedRerouteUsage()
  n11 = MaterialFunctionCall!(Vertex Mask=Add(A=Add(A=@n6, B=@n7), B=StaticSwitch(True=@n8, False=@n9, Value=@n10))) {materialFunction=/Game/Global/Materials/Character_Parent_2023/FlipbookFace_Parent_2023/MaterialFunctions/MF_Flipbook_VertColor_WPO}
  n12 = ScalarParameter() {
   controlType=Enumeration
   defaultValue=1
   enumeration=/Game/Global/Materials/Enums/E_Mat_OnOff
   group=Flipbook 01 Global
   parameterName=Flipbook.ApplyWPO
   sortPriority=1
  }
  NamedRerouteDeclaration(None=LinearInterpolate(A=Constant(), B=@n11.Flipbook WPO, Alpha=@n12)) {name=Flipbook_WPO, nodeColor=<rgba 1 0 0.0938 1>}
  NamedRerouteDeclaration(None=@n11.Vertex Color Hide Mask) {name=VertexColorHideMask, nodeColor=<rgba 1 0.545 0.1681 1>}
  n13 = NamedRerouteUsage()
  n14 = Constant()
  n15 = NamedRerouteUsage()
  n16 = Add(
   A=StaticSwitch(True=NamedRerouteUsage(), False=Constant(), Value=NamedRerouteUsage())
   B=StaticSwitch(True=Constant(), False=StaticSwitch(True=@n13, False=@n14, Value=@n15), Value=NamedRerouteUsage())
  )
  n17 = StaticSwitch(True=NamedRerouteUsage(), False=Constant(), Value=NamedRerouteUsage())
  NamedRerouteDeclaration(None=Saturate(None=Add(A=@n16, B=@n17))) {name=Flipbook_FaceMask, nodeColor=<rgba 1 0.7969 0 1>}
  n18 = StaticSwitch(True=NamedRerouteUsage(), False=Constant(), Value=NamedRerouteUsage())
  n19 = NamedRerouteUsage()
  n20 = NamedRerouteUsage()
  n21 = StaticSwitch(
   True=@n18
   False=StaticSwitch(True=LinearInterpolate(A=@n18, B=@n19, Alpha=@n20), False=@n18, Value=NamedRerouteUsage())
   Value=NamedRerouteUsage()
  )
  n22 = NamedRerouteUsage()
  n23 = NamedRerouteUsage()
  NamedRerouteDeclaration(None=StaticSwitch(True=LinearInterpolate(A=@n21, B=@n22, Alpha=@n23), False=@n21, Value=NamedRerouteUsage())) {name=Flipbook_FaceColor, nodeColor=<rgba 1 0.849 0.0986 1>}
  n24 = StaticBoolParameter() {
   desc="Needed for when additional areas of the mesh use vertex colors for hiding geo. Avoiding conflict with the filpbooks vertex mask functionality. "
   group=Flipbook 01 Global
   parameterName=Flipbook.UseExtraVertexMask
   sortPriority=-7
  }
  NamedRerouteDeclaration(None=@n24) {name=UseExtraVertexMask, nodeColor=<rgba 0.1953 1 0 1>}
  NamedRerouteDeclaration(None=ChannelMaskParameter(None=VertexColor()) {group=Flipbook 01 Global, parameterName=Flipbook.ExtraVertexColor, sortPriority=-6}.R) {name=Extra_VertexColor, nodeColor=<rgba 1 0.9844 0 1>}
  n25 = MaterialFunctionCall!(CustomUV) {materialFunction=/Game/Global/Materials/Character_Parent_2023/Anime_Parent_2023/MaterialFunctions/MFC_AnimeShader_2023_Pre}
  NamedRerouteDeclaration(None=@n25.Color_Lit) {name=Color_Lit, nodeColor=<rgba 0 1 0.8359 1>}
  NamedRerouteDeclaration(None=@n25.Color_Shaded) {name=Color_Shaded, nodeColor=<rgba 1 0.5156 0 1>}
  n26 = TextureSampleParameter2D() {
   group=Eye Animation 01 Global
   parameterName=EyeAnimation.EyeMask
   samplerSource=SSM_Wrap_WorldGroupSettings
   samplerType=SAMPLERTYPE_Masks
   sortPriority=3
   texture=&T_WhiteMask
  }
  n27 = VertexColor()
  n28 = ScalarParameter() {
   controlType=Enumeration
   enumeration=&E_Mat_OnOff
   group=Eye Animation 01 Global
   parameterName=EyeAnimation.UseVertexColorForEyeMask
   sortPriority=0
  }
  n29 = LinearInterpolate(A=@n26.RGBA, B=AppendVector(A=@n27, B=@n27), Alpha=@n28)
  n30 = TextureCoordinate() {coordinateIndex=1}
  n31 = Sign(None=Subtract(A=ComponentMask(None=@n30) {r=true}) {constB=0.0001})
  n32 = MaterialFunctionCall!(
   Color Lit=NamedRerouteUsage()
   Color Shaded=NamedRerouteUsage()
   Use UV2=StaticBoolParameter() {defaultValue=true, group=Eye Animation 01 Global, parameterName=EyeAnimation.UseUV2, sortPriority=-10}
   Mask=StaticSwitchParameter(
    True=ChannelMaskParameter(None=@n29) {group=Eye Animation 01 Global, parameterName=EyeAnimation.EyeMaskChannel, sortPriority=1}.R
    False=Saturate(None=@n31)
   ) {group=Eye Animation 01 Global, parameterName=EyeAnimation.UseEyeMask, sortPriority=0}
  ) {materialFunction=/CosmeticCompanions/Materials/MaterialFunctions/MF_SK_EyeAnimation}
  NamedRerouteDeclaration(None=StaticSwitch(True=@n32.Lit, False=NamedRerouteUsage(), Value=NamedRerouteUsage())) {name=EyeAnimColor_Lit, nodeColor=<rgba 0.1953 0 1 1>}
  NamedRerouteDeclaration(None=StaticSwitch(True=@n32.Shaded, False=NamedRerouteUsage(), Value=NamedRerouteUsage())) {name=EeyAnimColor_Shaded, nodeColor=<rgba 0.125 1 0 1>}
  n33 = MaterialFunctionCall!(Material=NamedRerouteUsage(), Vertex Color Hide Mask=NamedRerouteUsage()) {materialFunction=/Game/Global/Materials/Character_Parent_2023/FlipbookFace_Parent_2023/MaterialFunctions/MF_Flipbook_Debug}
  NamedRerouteDeclaration(None=@n33.Debug Vert Color) {name=DebugVertColor, nodeColor=<rgba 0 0.6953 1 1>}
  NamedRerouteDeclaration(None=@n33.Debug Vert Color WPO) {name=DebugVertColor_WPO, nodeColor=<rgba 1 0.9844 0 1>}
  NamedRerouteDeclaration(None=@n33.Debug UV WPO) {name=DebugUV_WPO, nodeColor=<rgba 1 0 0.3047 1>}
  NamedRerouteDeclaration(None=@n25.Emissive) {name=Emissive, nodeColor=<rgba 1 0.1875 0 1>}
  n34 = NamedRerouteUsage()
  n35 = Multiply(A=NamedRerouteUsage(), B=ScalarParameter() {defaultValue=1, group=Flipbook 01 Global, parameterName=Flipbook.EmissiveStrength})
  n36 = NamedRerouteUsage()
  n37 = ScalarParameter() {
   controlType=Enumeration
   enumeration=&E_Mat_OnOff
   group=Flipbook 01 Global
   parameterName=Flipbook.ApplyEmissive
   sortPriority=3
  }
  NamedRerouteDeclaration(None=LinearInterpolate(A=NamedRerouteUsage(), B=LinearInterpolate(A=@n34, B=@n35, Alpha=@n36), Alpha=@n37)) {name=Flipbook_Emissive, nodeColor=<rgba 0.108 0.2905 1 1>}
  n38 = ScalarParameter() {
   controlType=Enumeration
   enumeration=&E_Mat_OnOff
   group=Flipbook 01 Global
   parameterName=Flipbook.ApplyOpacityMask
   sortPriority=4
  }
  NamedRerouteDeclaration(None=LinearInterpolate(A=Constant(), B=NamedRerouteUsage(), Alpha=@n38)) {name=Flipbook_Opacity, nodeColor=<rgba 0 1 0.6016 1>}
  n39 = NamedRerouteUsage()
  n40 = Multiply(A=NamedRerouteUsage(), B=ScalarParameter() {defaultValue=1, parameterName=FlipbookFaceAnimeShadedValue})
  n41 = NamedRerouteUsage()
  n42 = ScalarParameter() {
   controlType=Enumeration
   defaultValue=1
   enumeration=&E_Mat_OnOff
   group=Flipbook 01 Global
   parameterName=Flipbook.ApplyBaseColor
   sortPriority=2
  }
  NamedRerouteDeclaration(None=LinearInterpolate(A=NamedRerouteUsage(), B=LinearInterpolate(A=@n39, B=@n40, Alpha=@n41), Alpha=@n42)) {name=Flipbook_ColorShaded, nodeColor=<rgba 1 0 0.0938 1>}
  n43 = NamedRerouteUsage()
  n44 = NamedRerouteUsage()
  n45 = NamedRerouteUsage()
  n46 = ScalarParameter() {
   controlType=Enumeration
   defaultValue=1
   enumeration=&E_Mat_OnOff
   group=Flipbook 01 Global
   parameterName=Flipbook.ApplyBaseColor
   sortPriority=2
  }
  NamedRerouteDeclaration(None=LinearInterpolate(A=NamedRerouteUsage(), B=LinearInterpolate(A=@n43, B=@n44, Alpha=@n45), Alpha=@n46)) {name=Flipbook_ColorLit, nodeColor=<rgba 0 0.4141 1 1>}
  NamedRerouteDeclaration(None=StaticBoolParameter() {group=Flipbook 06 Debug, parameterName=Flipbook.UseDebugVertColor, sortPriority=-1}) {name=UseDebugVertColor, nodeColor=<rgba 0.1953 1 0 1>}
  NamedRerouteDeclaration(None=StaticBoolParameter() {group=Flipbook 06 Debug, parameterName=Flipbook.UseDebugUVs, sortPriority=-1}) {name=UseDebugUVs, nodeColor=<rgba 0.1953 1 0 1>}
  NamedRerouteDeclaration(None=StaticBoolParameter() {group=Eye Animation 01 Global, parameterName="UseEyeAnimation (Overrides Flipbook Eyes)", sortPriority=-30}) {name=UseEyeAnimation, nodeColor=<rgba 0.1953 1 0 1>}
  NamedRerouteDeclaration(None=MaterialFunctionCall!(
    Index=MaterialFunctionCall() {materialFunction=&/MF_SK_Flipbook_Mouth_Curves_TEMP}.Index
    Input_0-3=NamedRerouteUsage()
    Input_4-7=NamedRerouteUsage()
   ) {materialFunction=/Game/Global/Materials/Character_Parent_2023/FlipbookFace_Parent_2023/MaterialFunctions/MF_RemapIndex_8}.Remapped Index) {name=Mouth_CurveIndex, nodeColor=<rgba 1 0.4453 0 1>}
  n47 = VectorParameter() {
   group=Flipbook 03 Eyes
   parameterName=Flipbook.EyesRemap_0-3
   sortPriority=201
   channelNames {a=3 - Lower Lid Up, b=2 - Wide, g=1 - Blink, r=0 - Default}
   defaultValue=<rgba 0 1 2 3>
  }
  NamedRerouteDeclaration(None=@n47.RGBA) {name=Eyes_Remap_0-3, nodeColor=<rgba 1 0.6652 0.0014 1>}
  n48 = VectorParameter() {
   group=Flipbook 03 Eyes
   parameterName=Flipbook.EyesRemap_4-7
   sortPriority=202
   channelNames {a=7 - Unused, b=6 - Unused, g=5 - Emote, r=4 - Upper Lid Down}
   defaultValue=<rgba 4 5 6 7>
  }
  NamedRerouteDeclaration(None=@n48.RGBA) {name=Eyes_Remap_4-7, nodeColor=<rgba 1 0.6652 0.0014 1>}
  n49 = VectorParameter() {
   group=Flipbook 02 Brows
   parameterName=Flipbook.BrowsRemap_0-3
   sortPriority=201
   channelNames {a=3 - Brows Out Down, b=2 - Brow In Up, g=1 - Brow In Down, r=0 - Default}
   defaultValue=<rgba 0 1 2 3>
  }
  NamedRerouteDeclaration(None=@n49.RGBA) {name=Brows_Remap_0-3, nodeColor=<rgba 1 0.6652 0.0014 1>}
  n50 = VectorParameter() {
   group=Flipbook 02 Brows
   parameterName=Flipbook.BrowsRemap_4-7
   sortPriority=202
   channelNames {a=7 - Unused, b=6 - Unused, g=5 - Emote, r=4 - Brow Out Up}
   defaultValue=<rgba 4 5 6 7>
  }
  NamedRerouteDeclaration(None=@n50.RGBA) {name=Brows_Remap_4-7, nodeColor=<rgba 1 0.6652 0.0014 1>}
  NamedRerouteDeclaration(None=MaterialFunctionCall!(
    Index=MaterialFunctionCall() {materialFunction=&/MF_SK_Flipbook_BrowAsEyes_Curves_TEMP}.Index
    Input_0-3=NamedRerouteUsage()
    Input_4-7=NamedRerouteUsage()
    Input_8-11=NamedRerouteUsage()
   ) {materialFunction=/Game/Global/Materials/Character_Parent_2023/FlipbookFace_Parent_2023/MaterialFunctions/MF_RemapIndex_12}.Remapped Index) {name=BrowsEyes_CurveIndex, nodeColor=<rgba 1 0.4453 0 1>}
  NamedRerouteDeclaration(None=StaticBoolParameter() {group=Flipbook 02 Brows, parameterName=Flipbook.UseTextureForBrows, sortPriority=-20}) {name=UseTextureForBrows, nodeColor=<rgba 0 1 0.3438 1>}
  n51 = TextureObjectParameter() {
   group=Flipbook 02 Brows
   parameterName=Flipbook.BrowsTexture
   sortPriority=2
   texture=/Game/Global/Materials/Character_Parent_2023/FlipbookFace_Parent_2023/Textures/T_black_emptyalpha
  }
  NamedRerouteDeclaration(None=@n51) {name=Brows_FlipbookTexture, nodeColor=<rgba 1 0.1875 0 1>}
  NamedRerouteDeclaration(None=ScalarParameter() {defaultValue=3, group=Flipbook 02 Brows, parameterName=Flipbook.BrowsRowCount, sortPriority=3}) {name=Brows_RowCount, nodeColor=<rgba 1 0 0.7734 1>}
  NamedRerouteDeclaration(None=ScalarParameter() {defaultValue=3, group=Flipbook 02 Brows, parameterName=Flipbook.BrowsColumnCount, sortPriority=4}) {name=Brows_ColumnCount, nodeColor=<rgba 1 0 0.7734 1>}
  NamedRerouteDeclaration(None=StaticSwitchParameter(True=TextureCoordinate() {coordinateIndex=1}, False=TextureCoordinate()) {group=Flipbook 02 Brows, parameterName=Flipbook.BrowsUseUV2, sortPriority=10}) {name=Brows_UVs, nodeColor=<rgba 1 0.5859 0 1>}
  n52 = ScalarParameter() {defaultValue=1, group=Flipbook 02 Brows, parameterName=Flipbook.BrowsUVScaleX, sortPriority=12}
  n53 = ScalarParameter() {defaultValue=1, group=Flipbook 02 Brows, parameterName=Flipbook.BrowsUVScaleY, sortPriority=13}
  NamedRerouteDeclaration(None=Multiply(
    A=ScalarParameter() {defaultValue=1, group=Flipbook 02 Brows, parameterName=Flipbook.BrowsGlobalScale, sortPriority=12}
    B=AppendVector(A=@n52, B=@n53)
   )) {name=Brows_UVScale, nodeColor=<rgba 0.2656 0 1 1>}
  n54 = VectorParameter() {
   group=Flipbook 03 Eyes
   parameterName=Flipbook.BrowsEyesRemap_4-7
   sortPriority=202
   channelNames {a=7 - Brows Out Down, b=6 - Brow In Up, g=5 - Brow In Down, r=4 - Upper Lid Down}
   defaultValue=<rgba 4 5 6 7>
  }
  NamedRerouteDeclaration(None=@n54.RGBA) {name=BrowsEyes_Remap_4-7, nodeColor=<rgba 1 0.6652 0.0014 1>}
  n55 = VectorParameter() {
   group=Flipbook 03 Eyes
   parameterName=Flipbook.EyesRemap_8-16
   sortPriority=202
   channelNames {a=11 - Unused, b=10 - Unused, g=9 - Emote, r=8 - Brows Out Up}
   defaultValue=<rgba 8 9 10 11>
  }
  NamedRerouteDeclaration(None=@n55.RGBA) {name=BrowsEyes_Remap_8-11, nodeColor=<rgba 1 0.6652 0.0014 1>}
  NamedRerouteDeclaration(None=AppendVector(
    A=ScalarParameter() {group=Flipbook 02 Brows, parameterName=Flipbook.BrowsUVOffsetX, sortPriority=20}
    B=ScalarParameter() {group=Flipbook 02 Brows, parameterName=Flipbook.BrowsUVOffsetY, sortPriority=21}
   )) {name=Brows_UVOffset, nodeColor=<rgba 0.0547 1 0 1>}
  NamedRerouteDeclaration(None=MaterialFunctionCall(
    Index=MaterialFunctionCall() {materialFunction=&/MF_SK_Flipbook_Brows_Curves_TEMP}.Index
    Input_0-3=NamedRerouteUsage()
    Input_4-7=NamedRerouteUsage()
   ) {materialFunction=&MF_RemapIndex_8}.Remapped Index) {name=Brows_CurveIndex, nodeColor=<rgba 1 0.4453 0 1>}
  NamedRerouteDeclaration(None=MaterialFunctionCall(
    Index=MaterialFunctionCall() {materialFunction=&/MF_SK_Flipbook_Eyes_Curves_TEMP}.Index
    Input_0-3=NamedRerouteUsage()
    Input_4-7=NamedRerouteUsage()
   ) {materialFunction=&MF_RemapIndex_8}.Remapped Index) {name=Eyes_CurveIndex, nodeColor=<rgba 1 0.4453 0 1>}
  NamedRerouteDeclaration(None=StaticBoolParameter() {group=Flipbook 03 Eyes, parameterName=Flipbook.UseTextureForEyes, sortPriority=-20}) {name=UseTextureForEyes, nodeColor=<rgba 0 1 0.3438 1>}
  NamedRerouteDeclaration(None=ScalarParameter() {defaultValue=3, group=Flipbook 03 Eyes, parameterName=Flipbook.EyesRowCount, sortPriority=3}) {name=Eyes_RowCount, nodeColor=<rgba 1 0 0.7734 1>}
  NamedRerouteDeclaration(None=ScalarParameter() {defaultValue=3, group=Flipbook 03 Eyes, parameterName=Flipbook.EyesColumnCount, sortPriority=4}) {name=Eyes_ColumnCount, nodeColor=<rgba 1 0 0.7734 1>}
  NamedRerouteDeclaration(None=StaticSwitchParameter(True=TextureCoordinate() {coordinateIndex=1}, False=TextureCoordinate()) {group=Flipbook 03 Eyes, parameterName=Flipbook.EyesUseUV2, sortPriority=10}) {name=Eyes_UVs, nodeColor=<rgba 1 0.5859 0 1>}
  n56 = ScalarParameter() {defaultValue=1, group=Flipbook 03 Eyes, parameterName=Flipbook.EyesUVScaleX, sortPriority=12}
  n57 = ScalarParameter() {defaultValue=1, group=Flipbook 03 Eyes, parameterName=Flipbook.EyesUVScaleY, sortPriority=13}
  NamedRerouteDeclaration(None=Multiply(
    A=ScalarParameter() {defaultValue=1, group=Flipbook 03 Eyes, parameterName=Flipbook.EyesGlobalScale, sortPriority=12}
    B=AppendVector(A=@n56, B=@n57)
   )) {name=Eyes_UVScale, nodeColor=<rgba 0.2656 0 1 1>}
  NamedRerouteDeclaration(None=AppendVector(
    A=ScalarParameter() {group=Flipbook 03 Eyes, parameterName=Flipbook.EyesUVOffsetX, sortPriority=20}
    B=ScalarParameter() {group=Flipbook 03 Eyes, parameterName=Flipbook.EyesUVOffsetY, sortPriority=21}
   )) {name=Eyes_UVOffset, nodeColor=<rgba 0.0547 1 0 1>}
  NamedRerouteDeclaration(None=ChannelMaskParameter(None=VertexColor()) {group=Flipbook 03 Eyes, parameterName=Flipbook.EyesVertexColor}.R) {name=Eyes_VertexColor, nodeColor=<rgba 1 0.9844 0 1>}
  n58 = TextureObjectParameter() {
   group=Flipbook 03 Eyes
   parameterName=Flipbook.EyesTexture
   sortPriority=2
   texture=&T_black_emptyalpha
  }
  NamedRerouteDeclaration(None=@n58) {name=Eyes_FlipbookTexture, nodeColor=<rgba 1 0.1875 0 1>}
  n59 = VectorParameter() {
   group=Flipbook 04 Mouth
   parameterName=Flipbook.MouthRemap_0-3
   sortPriority=201
   channelNames {a=3 - Jaw Open, b=2 - Frown, g=1 - Smile, r=0 - Default}
   defaultValue=<rgba 0 1 2 3>
  }
  NamedRerouteDeclaration(None=@n59.RGBA) {name=Mouth_Remap_0-3, nodeColor=<rgba 1 0.6652 0.0014 1>}
  NamedRerouteDeclaration(None=ChannelMaskParameter(None=VertexColor()) {group=Flipbook 04 Mouth, parameterName=Flipbook.MouthVertexColor}.R) {name=Mouth_VertexColor, nodeColor=<rgba 1 0.9844 0 1>}
  NamedRerouteDeclaration(None=StaticBoolParameter() {defaultValue=true, group=Flipbook 04 Mouth, parameterName=Flipbook.UseTextureForMouth, sortPriority=-20}) {name=UseTextureForMouth, nodeColor=<rgba 0 1 0.3438 1>}
  n60 = TextureObjectParameter() {
   group=Flipbook 04 Mouth
   parameterName=Flipbook.MouthTexture
   sortPriority=2
   texture=&T_black_emptyalpha
  }
  NamedRerouteDeclaration(None=@n60) {name=Mouth_FlipbookTexture, nodeColor=<rgba 1 0.1875 0 1>}
  NamedRerouteDeclaration(None=ScalarParameter() {defaultValue=3, group=Flipbook 04 Mouth, parameterName=Flipbook.MouthRowCount, sortPriority=3}) {name=Mouth_RowCount, nodeColor=<rgba 1 0 0.7734 1>}
  NamedRerouteDeclaration(None=ScalarParameter() {defaultValue=3, group=Flipbook 04 Mouth, parameterName=Flipbook.MouthColumnCount, sortPriority=4}) {name=Mouth_ColumnCount, nodeColor=<rgba 1 0 0.7734 1>}
  NamedRerouteDeclaration(None=StaticSwitchParameter(True=TextureCoordinate() {coordinateIndex=1}, False=TextureCoordinate()) {group=Flipbook 04 Mouth, parameterName=Flipbook.MouthUseUV2, sortPriority=10}) {name=Mouth_UVs, nodeColor=<rgba 1 0.5859 0 1>}
  n61 = ScalarParameter() {defaultValue=1, group=Flipbook 04 Mouth, parameterName=Flipbook.MouthUVScaleX, sortPriority=12}
  n62 = ScalarParameter() {defaultValue=1, group=Flipbook 04 Mouth, parameterName=Flipbook.MouthUVScaleY, sortPriority=13}
  NamedRerouteDeclaration(None=Multiply(
    A=ScalarParameter() {defaultValue=1, group=Flipbook 04 Mouth, parameterName=Flipbook.MouthGlobalScale, sortPriority=12}
    B=AppendVector(A=@n61, B=@n62)
   )) {name=Mouth_UVScale, nodeColor=<rgba 0.2656 0 1 1>}
  NamedRerouteDeclaration(None=AppendVector(
    A=ScalarParameter() {group=Flipbook 04 Mouth, parameterName=Flipbook.MouthUVOffsetX, sortPriority=20}
    B=ScalarParameter() {group=Flipbook 04 Mouth, parameterName=Flipbook.MouthUVOffsetY, sortPriority=21}
   )) {name=Mouth_UVOffset, nodeColor=<rgba 0.0547 1 0 1>}
  n63 = Constant() {r=1}
  n64 = NamedRerouteUsage()
  n65 = ScalarParameter() {defaultValue=0.1, group=Flipbook 01 Global, parameterName=Flipbook.ExtraVertexColorValue, sortPriority=-6}
  NamedRerouteDeclaration(None=LinearInterpolate(
    A=MaterialFunctionCall!(Hide Element Toggle=@n63, VertexColorChannel=@n64, Vertex Value to hide=@n65) {materialFunction=/Game/Packages/Fortress_SharedMaterials/Base_Material_Functions/VertexColorMask}.Result
    B=Constant()
    Alpha=ScalarParameter() {desc=Set by CID, group=Flipbook 01 Global, parameterName=Flipbook.HideExtraVertexColor, sortPriority=-5}
   )) {name=ExtraVertexMask, nodeColor=<rgba 0 1 0.3906 1>}
  NamedRerouteDeclaration(None=StaticBoolParameter() {group=Flipbook 01 Global, parameterName=Flipbook.UseBrows, sortPriority=-10}) {name=UseBrows, nodeColor=<rgba 0.1953 1 0 1>}
  n66 = MaterialFunctionCall(
   Curve Index=NamedRerouteUsage()
   Use Texture=NamedRerouteUsage()
   Vertex Color=NamedRerouteUsage()
   Flipbook Texture=NamedRerouteUsage()
   Column Count=NamedRerouteUsage()
   Row Count=NamedRerouteUsage()
   UVs=NamedRerouteUsage()
   UV Offset=NamedRerouteUsage()
   UV Scale=NamedRerouteUsage()
  ) {materialFunction=&MF_Sample_Flipbook}
  NamedRerouteDeclaration(None=@n66.Flipbook Vertex Mask) {name=BrowsVertexMask, nodeColor=<rgba 1 0 0.8438 1>}
  NamedRerouteDeclaration(None=@n66.Flipbook Color) {name=BrowsColor, nodeColor=<rgba 1 0 0.8438 1>}
  NamedRerouteDeclaration(None=@n66.Flipbook Alpha) {name=BrowsAlpha, nodeColor=<rgba 1 0 0.8438 1>}
  n67 = MaterialFunctionCall(
   Curve Index=StaticSwitchParameter(True=NamedRerouteUsage(), False=NamedRerouteUsage()) {group=Flipbook 01 Global, parameterName=FlipBook.UseBrowsAsEyes, sortPriority=-8}
   Use Texture=NamedRerouteUsage()
   Vertex Color=NamedRerouteUsage()
   Flipbook Texture=NamedRerouteUsage()
   Column Count=NamedRerouteUsage()
   Row Count=NamedRerouteUsage()
   UVs=NamedRerouteUsage()
   UV Offset=NamedRerouteUsage()
   UV Scale=NamedRerouteUsage()
  ) {materialFunction=&MF_Sample_Flipbook}
  NamedRerouteDeclaration(None=@n67.Flipbook Vertex Mask) {name=EyesVertexMask, nodeColor=<rgba 0.2436 0 1 1>}
  NamedRerouteDeclaration(None=@n67.Flipbook Color) {name=EyesColor, nodeColor=<rgba 0.2436 0 1 1>}
  NamedRerouteDeclaration(None=@n67.Flipbook Alpha) {name=EyesAlpha, nodeColor=<rgba 0.2436 0 1 1>}
  n68 = StaticBoolParameter() {
   defaultValue=true
   desc="Curves are shared with the material Anime Eye Animation functionality(found in \"01_EyeAnimationPoses\"). When the material Eye Animation is enabled, flipbook eyes will be disabled."
   group=Flipbook 01 Global
   parameterName=Flipbook.UseEyes
   sortPriority=-9
  }
  NamedRerouteDeclaration(None=@n68) {name=UseEyes, nodeColor=<rgba 0.1953 1 0 1>}
  n69 = MaterialFunctionCall!(
   CustomLightVector
   OverrideNormal_WS=MaterialFunctionCall!(Normal=@n25.Normal) {materialFunction=/CosmeticCompanions/Materials/MaterialFunctions/MF_CompanionAnime_NormalSphere}.Normal
   STT=@n25.STT
   Normal_Map=@n25.STT
   DFL=@n25.STT
   FX=@n25.STT
  ) {materialFunction=/Game/Global/Materials/Character_Parent_2023/Anime_Parent_2023/MaterialFunctions/MFC_AnimeShader_2023_Calculations}
  n70 = GetMaterialAttributes(None=MaterialFunctionCall!(
    Shading_Gradient=@n69.Shading_Gradient
    Threshold_Gradient=@n69.Shading_Gradient
    Specular_Gradient=@n69.Shading_Gradient
    Sheen_Mask=@n69.Shading_Gradient
    Thickness=@n69.Shading_Gradient
    DFL_Mask=@n69.Shading_Gradient
    FX=@n69.Shading_Gradient
    Metallic_Gradient=@n69.Shading_Gradient
    RimLight_Gradient=@n69.Shading_Gradient
    NormalBuffer=@n69.Shading_Gradient
    --------
    CINE_LightSources=@n69.CINE_LightSources
    CINE_LightColor=@n69.CINE_LightColor
    ---------
    Color_Lit=NamedRerouteUsage()
    Color_Shaded=NamedRerouteUsage()
    SRM=@n25.SRM
    Emissive=NamedRerouteUsage()
    DFL_Color=@n25.DFL_Color
    Specular_Color=@n25.Specular_Color
    RimLight_Color=@n25.StyleRimLight_Color
   ) {materialFunction=/Game/Global/Materials/Character_Parent_2023/Anime_Parent_2023/MaterialFunctions/MFC_AnimeShader_2023_Post}.Result) {attributeGetTypes [#1, #2]}
  n71 = NamedRerouteUsage()
  n72 = Add(A=@n70.World Position Offset, B=NamedRerouteUsage())
  n73 = NamedRerouteUsage()
  SetMaterialAttributes!(
   MaterialAttributes=NamedRerouteDeclaration(None=@n70.MaterialAttributes) {name=Material, nodeColor=<rgba 0.6406 0 1 1>}
   Base Color=StaticSwitch(True=NamedRerouteUsage(), False=@n70.Base Color, Value=NamedRerouteUsage())
   Opacity Mask=NamedRerouteUsage()
   World Position Offset=StaticSwitch(True=NamedRerouteUsage(), False=StaticSwitch(True=@n71, False=@n72, Value=@n73), Value=NamedRerouteUsage())
  ) {attributeSetTypes [#1, #3, #2]}
  TemporalResponsivenessOutput!(None=LinearInterpolate(B=ScalarParameter() {parameterName=Temporal Responsiveness}, Alpha=NamedRerouteUsage()) {constA=0.4})
 ]
}
